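# Conversation database for the Dark Elf NPC, providing its lore-flavored dialogue and access to shared NPC options.
# The Conversation.Entry.DarkElf entry feeds one Speech node, which has a default line and contextual messages.
# A first-ever meeting line uses a named priority curve, while five other lines check per-match conversation count for this NPC.
# Speech then branches through link nodes into the shared NPCLibrary Quests and BasicServices conversations.
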ConversationDatabase /NPCLibrary/NPCs/DarkElf/NPCLibrary_Conversation_DarkElf
entry_points [{tag=Conversation.Entry.DarkElf, destinations [#1]}]
nodes [
 ConversationEntryPointNode_7 {
  id=#1
  class=ConversationEntryPointNode
  outputs [#2]
  properties {entryTag {tagName=Conversation.Entry.DarkElf}}
 }
 Speech_C_4 {
  id=#2
  class=Speech_C
  outputs [#3, #4]
  properties {
   nodeName=""
   generalConfig {
    defaultMessage=Default conversation.
    contextualMessages [
     {
      message="Hail, from the Fallen Oasis."
      contextRequirements [{requirement {class=FortControllerRequirement_IsFirstEverConversationWithNPC}}]
      priority {curve {rowName=Tandem.Speech.Priority.FirstEverInteraction}}
     }
     {
      message=There is a greater power at play here than mere gods...
      contextRequirements [{requirement {class=FortControllerRequirement_CheckNPCMatchConversationCount, properties {nPCUniqueIdTag {tagName=AISpawnerData.Type.Tandem.DarkElf}}}}]
      priority {curve {rowName=None}}
     }
     {
      message=My souls see a warrior within you.
      contextRequirements [{requirement {class=FortControllerRequirement_CheckNPCMatchConversationCount, properties {nPCUniqueIdTag {tagName=AISpawnerData.Type.Tandem.DarkElf}}}}]
      priority {curve {rowName=None}}
     }
     {
      message="These gods are like any other powerful being: breakable, with the right tool."
      contextRequirements [{requirement {class=FortControllerRequirement_CheckNPCMatchConversationCount, properties {nPCUniqueIdTag {tagName=AISpawnerData.Type.Tandem.DarkElf}}}}]
      priority {curve {rowName=None}}
     }
     {
      message="My oasis was made a wasteland, eons ago. Beware dark strangers."
      contextRequirements [{requirement {class=FortControllerRequirement_CheckNPCMatchConversationCount, properties {nPCUniqueIdTag {tagName=AISpawnerData.Type.Tandem.DarkElf}}}}]
      priority {curve {rowName=None}}
     }
     {
      message=I seek rest after multiple lifetimes of strife. If peace is even possible...
      contextRequirements [{requirement {class=FortControllerRequirement_CheckNPCMatchConversationCount, properties {nPCUniqueIdTag {tagName=AISpawnerData.Type.Tandem.DarkElf}}}}]
      priority {curve {rowName=None}}
     }
    ]
   }
  }
 }
 ConversationLinkNode_0 {id=#4, class=ConversationLinkNode, properties {remoteEntryTag {tagName=Conversation.Entry.BasicServices}}}
 ConversationLinkNode_1 {id=#3, class=ConversationLinkNode, properties {remoteEntryTag {tagName=Conversation.Entry.NPCLibrary.Quests}}}
]
